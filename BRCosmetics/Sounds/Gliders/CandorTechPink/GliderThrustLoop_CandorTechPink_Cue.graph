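# Looping thrust sound cue for the CandorTechPink glider cosmetic, using glider thrust attenuation and loop audio property settings.
# A mixer layers five looping waves: high, mid and low thrust bands, plus gear and flap layers reused from the CandorDark glider.
# Each layer passes through a continuous modulator driven by the AnyDirection parameter, which scales its pitch and volume within per-layer ranges.
# Each Random node wraps a single wave, so no variation is picked.
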SoundCue /BRCosmetics/Sounds/Gliders/CandorTechPink/GliderThrustLoop_CandorTechPink_Cue
root=Mixer_0
nodes [
 {
  id=Mixer_0
  children [ModulatorContinuous_0, ModulatorContinuous_1, ModulatorContinuous_2, ModulatorContinuous_3, ModulatorContinuous_4]
  properties {inputVolume [1,1,1,1,1]}
 }
 {
  id=ModulatorContinuous_0
  children [Random_0]
  properties {
   pitchModulationParams {default=0, maxOutput=1.2, minOutput=1, parameterName=AnyDirection}
   volumeModulationParams {default=0, maxOutput=0.6, minOutput=0.4, parameterName=AnyDirection}
  }
 }
 {
  id=ModulatorContinuous_1
  children [Random_1]
  properties {
   pitchModulationParams {default=0, maxOutput=1.2, minOutput=1, parameterName=AnyDirection}
   volumeModulationParams {default=0, maxOutput=0.8, minOutput=0.3, parameterName=AnyDirection}
  }
 }
 {
  id=ModulatorContinuous_2
  children [Random_2]
  properties {
   pitchModulationParams {default=0, maxOutput=1.2, minOutput=1, parameterName=AnyDirection}
   volumeModulationParams {default=0, maxOutput=0.6, minOutput=0.2, parameterName=AnyDirection}
  }
 }
 {
  id=ModulatorContinuous_3
  children [Random_3]
  properties {
   pitchModulationParams {default=0, maxOutput=1.1, minOutput=1, parameterName=AnyDirection}
   volumeModulationParams {default=0, maxOutput=0.8, minOutput=0.4, parameterName=AnyDirection}
  }
 }
 {
  id=ModulatorContinuous_4
  children [Random_4]
  properties {
   pitchModulationParams {default=0, maxOutput=1.1, minOutput=1, parameterName=AnyDirection}
   volumeModulationParams {default=0, maxOutput=0.8, minOutput=0.4, parameterName=AnyDirection}
  }
 }
 {id=Random_0, children [WavePlayer_0], properties {weights [1]}}
 {id=Random_1, children [WavePlayer_1], properties {weights [1]}}
 {id=Random_2, children [WavePlayer_2], properties {weights [1]}}
 {id=Random_3, children [WavePlayer_3], properties {weights [1]}}
 {id=Random_4, children [WavePlayer_4], properties {weights [1]}}
 {
  id=WavePlayer_0
  properties {bLooping=true, soundWaveAssetPtr=&/Glider_CandorTechPink_Thrust_LP_High}
 }
 {
  id=WavePlayer_1
  properties {bLooping=true, soundWaveAssetPtr=&/Glider_CandorTechPink_Thrust_LP_Mid}
 }
 {
  id=WavePlayer_2
  properties {bLooping=true, soundWaveAssetPtr=&/Glider_CandorTechPink_Thrust_LP_Low}
 }
 {id=WavePlayer_3, properties {bLooping=true, soundWaveAssetPtr=/BRCosmetics/Sounds/Gliders/CandorDark/Glider_CandorDark_Thrust_Gear_LP}}
 {id=WavePlayer_4, properties {bLooping=true, soundWaveAssetPtr=/BRCosmetics/Sounds/Gliders/CandorDark/Glider_CandorDark_Thrust_Flap_LP}}
]
properties {
 attenuationSettings=&Fort_Player_Glider_Thrust_Athena_Att
 audioPropertiesSheet=&APS_Cosmetic_Glider_Loop
 maxDistance=12500
 soundClassObject=&Fort_Gliders_Land
 volumeMultiplier=0.9
 concurrencySet [null]
}
